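# Material function that shades the bubble effect on the Deranged Mile backpack cosmetic.
# It turns a height input into a smoothed normal, whose tangent-space XY samples a matcap texture.
# Matcap and input Fresnel, each scaled, add to a shifted world-up normal component and tint a reflection color to form Bubble Colors.
# The same tangent-space XY is also output as Bubble Warp.
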MaterialFunction /BRCosmetics/Accessories/FORT_Backpacks/Backpack_M_MED_DerangedMile/Materials/MaterialFunctions/MF_Backpack_DerangedMile_Bubble_Shading
graph {nodes [
  n1 = VectorParameter() {group=Bubbles - Color, parameterName=Reflection Color Down, sortPriority=7, defaultValue=<rgba 1.5 1.5 1.5 1>}
  n2 = MaterialFunctionCall!(Height=FunctionInput() {inputName=Height, inputType=FunctionInput_Scalar}, Absolute World Position, World Space Vertex Normals) {materialFunction=/Engine/Functions/Engine_MaterialFunctions03/Procedurals/HeightToNormalSmooth}
  n3 = ComponentMask(None=Transform(None=@n2.WorldSpaceNormal) {transformSourceType=TRANSFORMSOURCE_World, transformType=TRANSFORM_Tangent}) {g=true, r=true}
  n4 = TextureSampleParameter2D(UVs=Add(A=Multiply(A=@n3) {constB=0.5}) {constB=0.5}) {
   group=Bubbles - Color
   parameterName=Matcap
   sortPriority=-1
   texture=/BRCosmetics/Accessories/FORT_Backpacks/Backpack_M_MED_DerangedMile/Materials/T_Bubble_Matcap
  }
  n5 = Add(
   A=Multiply(A=@n4.R, B=ScalarParameter() {defaultValue=1.4, group=Bubbles - Color, parameterName=Matcap Intensity, sortPriority=9})
   B=Multiply(
    A=FunctionInput() {inputName=Fresnel, inputType=FunctionInput_Scalar}
    B=ScalarParameter() {defaultValue=1.4, group=Bubbles - Color, parameterName=Fresnel Intensity, sortPriority=8}
   )
  )
  n6 = Add(
   A=ComponentMask(None=@n2.WorldSpaceNormal) {b=true}
   B=ScalarParameter() {defaultValue=0.6, group=Bubbles - Color, parameterName=Bubbles Color Shift, sortPriority=8}
  ) {constB=0.5}
  FunctionOutput(None=Multiply(A=@n1.RGB, B=Add(A=@n5, B=@n6))) {outputName=Bubble Colors, sortPriority=1}
  FunctionOutput(None=@n3) {outputName=Bubble Warp, sortPriority=2}
 ]}
