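# Material function that draws animated, parallaxed bubbles in foreground and background layers for the Deranged Mile backpack, outputting color and a Fresnel-weighted mask.
# Each layer multiplies two panning bubble-texture samples, with positions offset along the local camera vector for fake depth and sine-based distortion.
# Contrasted panning noise products provide breakup masks, while a shared shading function derives bubble colors from the patterns.
# Inputs expose parallax offsets, occlusion masks and a vertical gradient.
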MaterialFunction /BRCosmetics/Accessories/FORT_Backpacks/Backpack_M_MED_DerangedMile/Materials/MaterialFunctions/MF_Backpack_DerangedMile_Bubbles
graph {nodes [
  n1 = NamedRerouteUsage()
  n2 = ScalarParameter() {defaultValue=0.5, group=Bubbles - Background, parameterName=BubblesOpacity_Background, sortPriority=0}
  n3 = NamedRerouteUsage()
  n4 = Multiply(A=@n3, B=@n3)
  n5 = Multiply(A=@n4, B=@n4)
  n6 = OneMinus(None=@n5)
  n7 = Saturate(None=Multiply(A=@n6) {constB=7})
  n8 = FunctionInput(Preview=Constant() {r=1}) {bUsePreviewValueAsDefault=true, inputName=Background Occlusion Mask, inputType=FunctionInput_Scalar, sortPriority=10, previewValue=<1 0 0 1>}
  n9 = NamedRerouteDeclaration(None=Multiply(A=Multiply(A=@n2, B=@n7), B=@n8)) {name=Background_Mask, nodeColor=<rgba 0.6172 1 0 1>}
  n10 = Multiply(A=NamedRerouteUsage(), B=ScalarParameter() {defaultValue=2, parameterName=B}) {constB=10}
  FunctionOutput(None=LinearInterpolate(A=Multiply(A=@n1, B=@n9), B=NamedRerouteUsage(), Alpha=Saturate(None=@n10)))
  n11 = NamedRerouteUsage()
  n12 = NamedRerouteUsage()
  n13 = Multiply(
   A=MaterialFunctionCall(Exponent=ScalarParameter() {defaultValue=3.5108, parameterName=Param_2}) {materialFunction=&AxisAlignedFresnel}.Fresnel
   B=Multiply(A=Max(A=@n11, B=@n12)) {constB=2}
  )
  FunctionOutput(None=NamedRerouteDeclaration(None=Saturate(None=@n13)) {name=Bubbles_Mask, nodeColor=<rgba 0 0.8828 1 1>}) {outputName=Mask, sortPriority=10}
  Multiply(A=@n6, B=NamedRerouteUsage())
  n14 = NamedRerouteUsage()
  n15 = Multiply(A=@n14, B=@n14)
  n16 = Multiply(A=@n15, B=@n15)
  n17 = OneMinus(None=@n16)
  Multiply(A=@n17, B=NamedRerouteUsage())
  n18 = TextureObjectParameter() {
   group=Bubbles - Global
   parameterName=BubbleMask
   samplerType=SAMPLERTYPE_LinearGrayscale
   sortPriority=-1
   texture=/Game/Environments/Asteria/Sets/Lair/Materials/LavaLamp/T_LavaLamp_Background
  }
  NamedRerouteDeclaration(None=@n18) {name=BubblesTex, nodeColor=<rgba 1 0.8672 0 1>}
  NamedRerouteDeclaration(None=ScalarParameter() {defaultValue=1, group=Bubbles - Global, parameterName=BubblesSpeed_Global, sortPriority=0}) {name=GlobalSpeed, nodeColor=<rgba 1 0 0.1641 1>}
  n19 = Multiply(
   A=ComponentMask(None=NamedRerouteUsage()) {b=true}
   B=ScalarParameter() {defaultValue=0.1, group=Bubbles - Global, parameterName=BubblesDistortTiling_Global, sortPriority=0}
  )
  n20 = Frac(None=Multiply(A=Time(), B=ScalarParameter() {defaultValue=0.15, group=Bubbles - Global, parameterName=BubblesDistortSpeed_Global, sortPriority=0}) {constB=0.5})
  NamedRerouteDeclaration(None=Sine(None=Add(A=@n19, B=@n20))) {name=GlobalDistort, nodeColor=<rgba 1 0 0.0469 1>}
  n21 = NamedRerouteUsage()
  n22 = VectorParameter() {defaultValue=<rgba 1 0.65 1 1>}
  n23 = Multiply(
   A=Multiply(A=ComponentMask(None=@n21) {b=true, r=true}, B=ComponentMask(None=@n22.RGB) {g=true, r=true})
   B=ScalarParameter() {defaultValue=-0.02, parameterName=BackgroundBubbleMaskScale}
  )
  n24 = ComponentMask(None=VectorParameter() {parameterName=BackgroundBubbleMaskSpeed, defaultValue=<rgba -0.01 0.1 0 1>}.RGB) {g=true, r=true}
  n25 = Panner(Coordinate=@n23, Speed=@n24)
  n26 = Panner(Coordinate=Multiply(A=@n23) {constB=1.5}, Speed=Multiply(A=@n24) {constB=0.75})
  n27 = Power(
   Base=Multiply(
    A=TextureSample(UVs=@n25) {samplerType=SAMPLERTYPE_Grayscale, texture=&LowResBlurredNoise}.R
    B=TextureSample(UVs=@n26) {samplerType=SAMPLERTYPE_Grayscale, texture=&LowResBlurredNoise}.R
   )
   Exp=ScalarParameter() {defaultValue=0.32, parameterName=BackgroundBubbleMaskExponent}
  )
  n28 = ScalarParameter() {defaultValue=20, parameterName="Contrast (S)", sliderMin=582.2456}
  NamedRerouteDeclaration(None=Saturate(None=MaterialFunctionCall(In=@n27, Contrast=@n28) {materialFunction=&CheapContrast}.Result)) {name=BackgroundMask, nodeColor=<rgba 0 1 0.2734 1>}
  n29 = NamedRerouteUsage()
  n30 = Subtract(A=@n29, B=TransformPosition(None=CameraPositionWS()) {transformSourceType=TRANSFORMPOSSOURCE_World})
  n31 = ScalarParameter() {defaultValue=8, group=Bubbles - Background, parameterName=BubblesParallaxDepth_Background, sortPriority=3}
  n32 = FunctionInput() {bUsePreviewValueAsDefault=true, inputName=Background Parallax Offset, inputType=FunctionInput_Scalar}
  n33 = Add(A=@n29, B=Multiply(A=Normalize(VectorInput=@n30), B=Add(A=@n31, B=@n32)))
  n34 = VectorParameter() {group=Bubbles - Background, parameterName=BubblesPositionOffset_1_Background, sortPriority=30, defaultValue=<rgba 0 0 0 1>}
  n35 = Add(
   A=NamedRerouteDeclaration(None=Add(A=@n33, B=@n34.RGB)) {name=BubbleWorldPosParallax_2, nodeColor=<rgba 0 1 0.3203 1>}
   B=NamedRerouteUsage()
  )
  n36 = Multiply(
   A=NamedRerouteUsage()
   B=ScalarParameter() {defaultValue=8, group=Bubbles - Background, parameterName=BubblesSize_1_Background, sortPriority=10}
  )
  n37 = ScalarParameter() {defaultValue=0.02, group=Bubbles - Background, parameterName=BubblesSpeedX_1_Background, sortPriority=20}
  n38 = ScalarParameter() {defaultValue=0.0346, group=Bubbles - Background, parameterName=BubblesSpeedY_1_Background, sortPriority=20}
  n39 = Panner(
   Coordinate=Divide(A=ComponentMask(None=@n35) {b=true, r=true}, B=Multiply(A=@n36) {constB=-1})
   Speed=Multiply(A=NamedRerouteUsage(), B=AppendVector(A=@n37, B=@n38))
  )
  n40 = NamedRerouteUsage()
  n41 = NamedRerouteUsage()
  n42 = Subtract(A=@n41, B=TransformPosition(None=CameraPositionWS()) {transformSourceType=TRANSFORMPOSSOURCE_World})
  n43 = ScalarParameter() {defaultValue=8, group=Bubbles - Background, parameterName=BubblesParallaxDepth_Background, sortPriority=3}
  n44 = Add(A=@n41, B=Multiply(A=Normalize(VectorInput=@n42), B=Add(A=@n43)))
  n45 = VectorParameter() {group=Bubbles - Background, parameterName=BubblesPositionOffset_2_Background, sortPriority=31, defaultValue=<rgba 0 0 0 1>}
  n46 = Add(
   A=NamedRerouteDeclaration(None=Add(A=@n44, B=@n45.RGB)) {name=BubbleWorldPosParallax_3, nodeColor=<rgba 0 1 0.3203 1>}
   B=NamedRerouteUsage()
  )
  n47 = Multiply(
   A=NamedRerouteUsage()
   B=ScalarParameter() {defaultValue=5, group=Bubbles - Background, parameterName=BubblesSize_2_Background, sortPriority=11}
  )
  n48 = ScalarParameter() {defaultValue=0.02, group=Bubbles - Background, parameterName=BubblesSpeedX_2_Background, sortPriority=21}
  n49 = ScalarParameter() {defaultValue=0.0952, group=Bubbles - Background, parameterName=BubblesSpeedY_2_Background, sortPriority=21}
  n50 = Panner(
   Coordinate=Divide(A=ComponentMask(None=@n46) {b=true, r=true}, B=Multiply(A=@n47) {constB=-1})
   Speed=Multiply(A=NamedRerouteUsage(), B=AppendVector(A=@n48, B=@n49))
  )
  n51 = NamedRerouteUsage()
  n52 = Multiply(A=Multiply(A=TextureSample(UVs=@n39, Tex=@n40).R, B=TextureSample(UVs=@n50, Tex=@n51).R)) {constB=1.75}
  n53 = NamedRerouteUsage()
  NamedRerouteDeclaration(None=Max(A=Saturate(None=@n52), B=OneMinus(None=@n53))) {name=Background_Pattern, nodeColor=<rgba 0 1 0.1094 1>}
  n54 = VectorParameter() {group=Bubbles - Background, parameterName=BubblesDistortAmplitude_Background, sortPriority=4, defaultValue=<rgba 0 0 0 1>}
  NamedRerouteDeclaration(None=Multiply(A=@n54.RGB, B=NamedRerouteUsage())) {name=BubbleDistort_3, nodeColor=<rgba 1 0 0.3281 1>}
  n55 = NamedRerouteUsage()
  n56 = NamedRerouteUsage()
  n57 = Subtract(A=@n56, B=TransformPosition(None=CameraPositionWS()) {transformSourceType=TRANSFORMPOSSOURCE_World})
  n58 = ScalarParameter() {defaultValue=4, group=Bubbles - Foreground, parameterName=BubblesParallaxDepth_Foreground, sortPriority=2}
  n59 = FunctionInput() {bUsePreviewValueAsDefault=true, inputName=Foreground Parallax Offset_1, inputType=FunctionInput_Scalar}
  n60 = Add(A=@n56, B=Multiply(A=Normalize(VectorInput=@n57), B=Add(A=@n58, B=@n59)))
  n61 = VectorParameter() {group=Bubbles - Foreground, parameterName=BubblesPositionOffset_1_Foregrund, sortPriority=30, defaultValue=<rgba 0 0 0 1>}
  n62 = Add(A=NamedRerouteDeclaration(None=Add(A=@n60, B=@n61.RGB)) {name=BubbleWorldPosParallax, nodeColor=<rgba 0 1 0.3203 1>}, B=NamedRerouteUsage())
  n63 = Multiply(
   A=NamedRerouteUsage()
   B=ScalarParameter() {defaultValue=10, group=Bubbles - Foreground, parameterName=BubblesSize_1_Foreground, sortPriority=10}
  )
  n64 = ScalarParameter() {defaultValue=0.01, group=Bubbles - Foreground, parameterName=BubblesSpeedX_1_Foreground, sortPriority=20}
  n65 = ScalarParameter() {defaultValue=0.0357, group=Bubbles - Foreground, parameterName=BubblesSpeedY_1_Foreground, sortPriority=20}
  n66 = Panner(
   Coordinate=Divide(A=ComponentMask(None=@n62) {b=true, r=true}, B=Multiply(A=@n63) {constB=-1})
   Speed=Multiply(A=NamedRerouteUsage(), B=AppendVector(A=@n64, B=@n65))
  )
  n67 = NamedRerouteUsage()
  n68 = NamedRerouteUsage()
  n69 = Subtract(A=@n68, B=TransformPosition(None=CameraPositionWS()) {transformSourceType=TRANSFORMPOSSOURCE_World})
  n70 = ScalarParameter() {defaultValue=4, group=Bubbles - Foreground, parameterName=BubblesParallaxDepth_Foreground, sortPriority=2}
  n71 = Add(A=@n68, B=Multiply(A=Normalize(VectorInput=@n69), B=Add(A=@n70, B=@n59)))
  n72 = VectorParameter() {group=Bubbles - Foreground, parameterName=BubblesPositionOffset_2_Foreground, sortPriority=31, defaultValue=<rgba 0 0 0 1>}
  n73 = Add(
   A=NamedRerouteDeclaration(None=Add(A=@n71, B=@n72.RGB)) {name=BubbleWorldPosParallax_4, nodeColor=<rgba 0 1 0.3203 1>}
   B=NamedRerouteUsage()
  )
  n74 = Multiply(
   A=NamedRerouteUsage()
   B=ScalarParameter() {defaultValue=7, group=Bubbles - Foreground, parameterName=BubblesSize_2_Foreground, sortPriority=11}
  )
  n75 = ScalarParameter() {defaultValue=-0.02, group=Bubbles - Foreground, parameterName=BubblesSpeedX_2_Foreground, sortPriority=21}
  n76 = ScalarParameter() {defaultValue=0.1465, group=Bubbles - Foreground, parameterName=BubblesSpeedY_2_Foreground, sortPriority=21}
  n77 = Panner(
   Coordinate=Divide(A=ComponentMask(None=@n73) {b=true, r=true}, B=Multiply(A=@n74) {constB=-1})
   Speed=Multiply(A=NamedRerouteUsage(), B=AppendVector(A=@n75, B=@n76))
  )
  n78 = NamedRerouteUsage()
  n79 = Multiply(A=Multiply(A=TextureSample(UVs=@n66, Tex=@n67).R, B=TextureSample(UVs=@n77, Tex=@n78).R)) {constB=2}
  NamedRerouteDeclaration(None=Max(A=OneMinus(None=@n55), B=Saturate(None=@n79))) {name=Foreground_Pattern, nodeColor=<rgba 1 0 0.1069 1>}
  n80 = ScalarParameter() {defaultValue=1, group=Bubbles - Foreground, parameterName=BubblesOpacity_Foreground, sortPriority=0}
  n81 = Saturate(None=Multiply(A=@n17) {constB=7})
  n82 = FunctionInput(Preview=Constant() {r=1}) {bUsePreviewValueAsDefault=true, inputName=Foreground Occlusion Mask, inputType=FunctionInput_Scalar, sortPriority=10, previewValue=<1 0 0 1>}
  NamedRerouteDeclaration(None=Multiply(A=Multiply(A=@n80, B=@n81), B=@n82)) {name=Foreground_Mask, nodeColor=<rgba 0.6172 1 0 1>}
  n83 = MaterialFunctionCall!(
   Fresnel=@n16
   Height=Multiply(
    A=OneMinus(None=@n14)
    B=ScalarParameter() {defaultValue=1, group=Bubbles - Foreground, parameterName=BubblesNormalIntensity_Foreground, sortPriority=1}
   )
  ) {materialFunction=&/MF_Backpack_DerangedMile_Bubble_Shading}
  NamedRerouteDeclaration(None=@n83.Bubble Colors) {name=Foreground_Color, nodeColor=<rgba 1 0 0.9375 1>}
  n84 = MaterialFunctionCall(
   Fresnel=@n5
   Height=Multiply(
    A=OneMinus(None=@n3)
    B=ScalarParameter() {defaultValue=1, group=Bubbles - Background, parameterName=BubblesNormalIntensity_Background, sortPriority=2}
   )
  ) {materialFunction=&MF_Backpack_DerangedMile_Bubble_Shading}
  NamedRerouteDeclaration(None=@n84.Bubble Colors) {name=Background_Color, nodeColor=<rgba 1 0 0.9375 1>}
  NamedRerouteDeclaration(None=FunctionInput() {inputName=VerticalGradient, inputType=FunctionInput_Scalar}) {name=VerticalGradient, nodeColor=<rgba 1 0.7969 0 1>}
  n85 = VectorParameter() {group=Bubbles - Foreground, parameterName=BubblesDistortAmplitude_Foreground, sortPriority=3, defaultValue=<rgba 0.2 0.2 0 1>}
  NamedRerouteDeclaration(None=Multiply(A=@n85.RGB, B=NamedRerouteUsage())) {name=BubbleDistort_2, nodeColor=<rgba 1 0 0.3281 1>}
  n86 = NamedRerouteUsage()
  n87 = VectorParameter() {defaultValue=<rgba 1 0.65 1 1>}
  n88 = Multiply(
   A=Multiply(A=ComponentMask(None=@n86) {b=true, r=true}, B=ComponentMask(None=@n87.RGB) {g=true, r=true})
   B=ScalarParameter() {defaultValue=0.03, parameterName=ForegroundBubbleMaskScale}
  )
  n89 = ComponentMask(None=VectorParameter() {parameterName=ForegroundBubbleMaskSpeed, defaultValue=<rgba 0.05 -0.1 0 1>}.RGB) {g=true, r=true}
  n90 = Panner(Coordinate=@n88, Speed=@n89)
  n91 = Panner(Coordinate=Multiply(A=@n88) {constB=1.5}, Speed=Multiply(A=@n89) {constB=0.75})
  n92 = Power(
   Base=Multiply(
    A=TextureSample(UVs=@n90) {samplerType=SAMPLERTYPE_Grayscale, texture=&LowResBlurredNoise}.R
    B=TextureSample(UVs=@n91) {samplerType=SAMPLERTYPE_Grayscale, texture=&LowResBlurredNoise}.R
   )
   Exp=ScalarParameter() {defaultValue=0.5, parameterName=ForegroundBubbleMaskExponent}
  )
  NamedRerouteDeclaration(None=Saturate(None=MaterialFunctionCall(In=@n92, Contrast=@n28) {materialFunction=&CheapContrast}.Result)) {name=ForegroundMask, nodeColor=<rgba 1 0 0.4687 1>}
  NamedRerouteDeclaration(None=VertexInterpolator(VS=LocalPosition() {localOrigin=InstancePreSkinning}.XYZ).PS) {name=PreSkinnedPos, nodeColor=<rgba 0.2422 0 1 1>}
  NamedRerouteDeclaration(None=ScalarParameter() {defaultValue=1, group=Bubbles - Global, parameterName=BubblesSize_Global, sortPriority=0}) {name=GlobalSize, nodeColor=<rgba 1 0 0.3984 1>}
  NamedRerouteUsage()
 ]}
