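# Material function for the glass of the Deranged Mile backpack, adding depth-based fog and bubbles to an input material's attributes.
# The difference between scene depth and pixel depth drives a min-to-max fog density mask and a smoothstep mask that occludes background bubbles.
# MF_Backpack_DerangedMile_Bubbles supplies bubble color and mask, and the output sets emissive, blended toward FogColor_Up, and opacity.
# The vertical gradient is a constant placeholder, and some debug mask nodes remain.
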MaterialFunction /BRCosmetics/Accessories/FORT_Backpacks/Backpack_M_MED_DerangedMile/Materials/MaterialFunctions/MF_Backpack_DerangedMile_Glass
graph {nodes [
  Constant() {r=1}
  n1 = NamedRerouteUsage()
  n2 = VectorParameter() {group=Fog, parameterName=FogColor_Up, defaultValue=<rgba 0 0 0 1>}
  n3 = Min(A=NamedRerouteUsage(), B=OneMinus(None=NamedRerouteUsage()))
  n4 = NamedRerouteUsage()
  FunctionOutput(None=SetMaterialAttributes!(MaterialAttributes=NamedRerouteUsage(), BaseColor, EmissiveColor=LinearInterpolate(A=@n1, B=@n2.RGB, Alpha=@n3), Opacity=@n4) {attributeSetTypes [#1, #2, #3]})
  Max(A=@n4, B=NamedRerouteUsage())
  n5 = Subtract(A=SceneDepth().R, B=PixelDepth().R)
  n6 = Divide(A=@n5, B=ScalarParameter() {defaultValue=78.7933, parameterName=FogDepth})
  NamedRerouteDeclaration(None=LinearInterpolate(
    A=ScalarParameter() {defaultValue=0.6, group=Fog, parameterName=Fog Min Density}
    B=ScalarParameter() {defaultValue=1, group=Fog, parameterName=Fog Max Density}
    Alpha=Saturate(None=@n6)
   ) {constA=0.5}) {name=FogMask, nodeColor=<rgba 1 0 0.6328 1>}
  NamedRerouteDeclaration(None=Constant()) {name=VerticalGradient, nodeColor=<rgba 1 0 0.9844 1>}
  n7 = Saturate(None=Divide(A=@n5, B=ScalarParameter() {defaultValue=5}))
  NamedRerouteDeclaration(None=SphereMask(A=@n7, B=ScalarParameter() {defaultValue=2.0036, parameterName=Param_3}, Radius=ScalarParameter() {defaultValue=1.1672, parameterName=wdqef}) {hardnessPercent=0}) {name=debug, nodeColor=<rgba 1 0 0.1641 1>}
  n8 = Divide(A=@n5, B=Constant() {r=12})
  n9 = MaterialFunctionCall!(
   Foreground Parallax Offset_1
   Background Parallax Offset=Constant()
   VerticalGradient=OneMinus(None=NamedRerouteUsage())
   Foreground Occlusion Mask
   Background Occlusion Mask=SmoothStep(Min=ScalarParameter() {defaultValue=0.41, parameterName=Param_2}, Max=Constant() {r=1}, Value=Saturate(None=@n8)) {desc=Ribs Mask}
  ) {materialFunction=&/MF_Backpack_DerangedMile_Bubbles}
  NamedRerouteDeclaration(None=@n9.Mask) {name=BubblesMask, nodeColor=<rgba 0 0.2031 1 1>}
  NamedRerouteDeclaration(None=@n9.Result) {name=BubblesColor, nodeColor=<rgba 0.5234 1 0 1>}
  n10 = FunctionInput() {inputName=InMat, inputType=FunctionInput_MaterialAttributes}
  NamedRerouteDeclaration(None=GetMaterialAttributes(None=NamedRerouteDeclaration(None=@n10) {name=InMat, nodeColor=<rgba 1 0.9375 0 1>}) {attributeGetTypes [#1, #2, #3]}.Opacity) {name=InOpacity, nodeColor=<rgba 1 0 0.8437 1>}
  NamedRerouteUsage()
  OneMinus(None=@n7)
  SmoothStep(Min=ScalarParameter() {parameterName=Param_1}, Max=ScalarParameter() {defaultValue=1, parameterName=Param_4}, Value=@n7) {desc=Ribs Mask}
 ]}
